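; Behavior tree service for Fortnite AI that sends gameplay event tags to blackboard-referenced actors, such as self or GoalActor.
; On activation it caches the pawn, GoalActor and ability system components, then sends each ActivationEvents tag to its mapped key's actor, provided GoalActor is valid.
; Tick repeats that same logic when UseEventTickLogic is set.
; Deactivation sends the DeactivationEvents map the same way, without the pawn or GoalActor checks.
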(event EventReceiveActivation
  (params (OwnerActor :type Actor%20Object%20Reference))
  (bind _activationeventvaluesblackboards (Variables|Default|GetActivationEventValuesBlackboards))
  (Variables|Default|SetOwnerActor OwnerActor)
  (bind _owneractor (Variables|Default|GetOwnerActor)
    (:then
      (bind _asfort_aicontroller (Utilities|Casting|CastToFortAIController _owneractor))
      (bind _myfortaipawn (Variables|/Script/FortniteAI.FortAIController|GetMyFortAIPawn _asfort_aicontroller)
        (:then
          (bind _output_get (Variables|Default|SetActivatingPawn _myfortaipawn))
          (Utilities|IsValid (Ability|GetAbilitySystemComponent _output_get)
            (:Is%20Valid
              (Variables|Default|SetActivatingPawnAbilitySystemComponent (Ability|GetAbilitySystemComponent _output_get))
              (Utilities|IsValid (AI|GetBlackboard _owneractor)
                (:Is%20Valid
                  (bind _asactor (Utilities|Casting|CastToActor (AI|Components|Blackboard|GetValueasObject (AI|GetBlackboard _owneractor) (Utilities|Name|MakeLiteralName "GoalActor"))))
                  (Utilities|IsValid _asactor
                    (:Is%20Valid
                      (Variables|Default|SetGoalActor _asactor)
                      (Utilities|IsValid (Ability|GetAbilitySystemComponent _asactor)
                        (:Is%20Valid
                          (Variables|Default|SetTargetAbilitySystemComponent (Ability|GetAbilitySystemComponent _asactor)))
                        (:Is%20Not%20Valid
                          _))
                      (bind _keys (Utilities|Map|Keys (Variables|Default|GetActivationEvents)))
                      (bind _output_get_2 (Variables|Default|SetActivationEventKeysTags _keys))
                      (bind _values (Utilities|Map|Values (Variables|Default|GetActivationEvents)))
                      (Variables|Default|SetActivationEventValuesBlackboards _values)
                      (for (_array_element _array_index) _output_get_2
                        (Utilities|IsValid (AI|BehaviorTree|GetBlackboardValueasActor (Utilities|Array|Get(acopy) _activationeventvaluesblackboards _array_index))
                          (:Is%20Valid
                            (Ability|SendGameplayEventtoActor (AI|BehaviorTree|GetBlackboardValueasActor (Utilities|Array|Get(acopy) _activationeventvaluesblackboards _array_index)) _array_element :!pinned true) ; On Activation send all of these Gameplay Event Tags to particular Blackboard Key Actors\n\nCould be Self, GoalActor, etc.
                          ))))))))))
        (:else)))
    (:else)))

(event EventReceiveTick
  (params (OwnerActor :type Actor%20Object%20Reference) (DeltaSeconds :type Float%20%28single-precision%29))
  (bind _activationeventvaluesblackboards (Variables|Default|GetActivationEventValuesBlackboards))
  (if (Variables|Service|GetUseEventTickLogic)
    (Variables|Default|SetOwnerActor OwnerActor)
    (bind _owneractor (Variables|Default|GetOwnerActor)
      (:then
        (bind _asfort_aicontroller (Utilities|Casting|CastToFortAIController _owneractor))
        (bind _myfortaipawn (Variables|/Script/FortniteAI.FortAIController|GetMyFortAIPawn _asfort_aicontroller)
          (:then
            (bind _output_get (Variables|Default|SetActivatingPawn _myfortaipawn))
            (Utilities|IsValid (Ability|GetAbilitySystemComponent _output_get)
              (:Is%20Valid
                (Variables|Default|SetActivatingPawnAbilitySystemComponent (Ability|GetAbilitySystemComponent _output_get))
                (Utilities|IsValid (AI|GetBlackboard _owneractor)
                  (:Is%20Valid
                    (bind _asactor (Utilities|Casting|CastToActor (AI|Components|Blackboard|GetValueasObject (AI|GetBlackboard _owneractor) (Utilities|Name|MakeLiteralName "GoalActor"))))
                    (Utilities|IsValid _asactor
                      (:Is%20Valid
                        (Variables|Default|SetGoalActor _asactor)
                        (Utilities|IsValid (Ability|GetAbilitySystemComponent _asactor)
                          (:Is%20Valid
                            (Variables|Default|SetTargetAbilitySystemComponent (Ability|GetAbilitySystemComponent _asactor)))
                          (:Is%20Not%20Valid
                            _))
                        (bind _keys (Utilities|Map|Keys (Variables|Default|GetActivationEvents)))
                        (bind _output_get_2 (Variables|Default|SetActivationEventKeysTags _keys))
                        (bind _values (Utilities|Map|Values (Variables|Default|GetActivationEvents)))
                        (Variables|Default|SetActivationEventValuesBlackboards _values)
                        (for (_array_element _array_index) _output_get_2
                          (Utilities|IsValid (AI|BehaviorTree|GetBlackboardValueasActor (Utilities|Array|Get(acopy) _activationeventvaluesblackboards _array_index))
                            (:Is%20Valid
                              (Ability|SendGameplayEventtoActor (AI|BehaviorTree|GetBlackboardValueasActor (Utilities|Array|Get(acopy) _activationeventvaluesblackboards _array_index)) _array_element :!pinned true) ; On Activation send all of these Gameplay Event Tags to particular Blackboard Key Actors\n\nCould be Self, GoalActor, etc.
                            ))))))))))
          (:else)))
      (:else))))

(event EventReceiveDeactivation
  (params (OwnerActor :type Actor%20Object%20Reference))
  (bind _deactivationeventsvaluesblackboards (Variables|Default|GetDeactivationEventsValuesBlackboards))
  (bind _keys (Utilities|Map|Keys (Variables|Default|GetDeactivationEvents)))
  (bind _output_get (Variables|Default|SetDeactivationEventKeysTags _keys))
  (bind _values (Utilities|Map|Values (Variables|Default|GetDeactivationEvents)))
  (Variables|Default|SetDeactivationEventsValuesBlackboards _values)
  (for (_array_element _array_index) _output_get
    (Utilities|IsValid (AI|BehaviorTree|GetBlackboardValueasActor (Utilities|Array|Get(acopy) _deactivationeventsvaluesblackboards _array_index))
      (:Is%20Valid
        (Ability|SendGameplayEventtoActor (AI|BehaviorTree|GetBlackboardValueasActor (Utilities|Array|Get(acopy) _deactivationeventsvaluesblackboards _array_index)) _array_element :!pinned true) ; On Deactivation send all of these Gameplay Event Tags to particular Blackboard Key Actors\n\nCould be Self, GoalActor, etc.
      ))))
